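# Sound cue for activating the Hightower Grape Bramble Shield gadget or ability.
# It plays a single wave, HightowerGrape_BrambleShield_Activate, through one wave player with no randomization or modulation nodes.
# Playback uses the basic Foley attenuation asset and the Fort_Gadgets sound class, with slightly reduced volume.
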SoundCue /HighTower/Audio/Grape/HightowerGrape_BrambleShield_Activate_Cue
root=WavePlayer_0
nodes [{id=WavePlayer_0, properties {soundWaveAssetPtr=&/HightowerGrape_BrambleShield_Activate}}]
properties {
 attenuationSettings=&Foley_Attenuation_Basic_Att
 maxDistance=1536
 soundClassObject=&Fort_Gadgets
 volumeMultiplier=0.8
}
