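# Sound cue for dropping the Grape Bramble Shield in Hightower, played as a spatialized loot sound.
# A mixer layers the shield's impact wave with a quieter, reused Tapas drop wave, then feeds a Modulator for per-play pitch and volume variation.
# The cue is pitched up and turned down overall, using weapon-drop attenuation and the loot-collect sound class.
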SoundCue /HighTower/Audio/Grape/HightowerGrape_BrambleShield_Drop_Cue
root=Modulator_1
nodes [
 {id=Mixer_0, children [WavePlayer_0,WavePlayer_5], properties {inputVolume [1,0.6]}}
 {id=Modulator_1, children [Mixer_0]}
 {id=WavePlayer_0, properties {soundWaveAssetPtr=&/HightowerGrape_BrambleShield_Impact_01}}
 {id=WavePlayer_5, properties {soundWaveAssetPtr=/HighTower/Audio/Tapas/Tapas_Drop_01}}
]
properties {
 attenuationSettings=&Fort_Weapon_Drop_Att
 maxDistance=3000
 pitchMultiplier=1.2
 soundClassObject=&Fort_Loot_Collect
 volumeMultiplier=0.7
 concurrencySet [null]
}
