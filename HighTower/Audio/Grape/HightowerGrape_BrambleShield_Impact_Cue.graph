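# Sound cue for impacts on the Hightower Grape Bramble Shield.
# A Random node picks one of four equally weighted impact wave variations, with the node excluded from branch culling.
# Playback runs at reduced volume through the gadgets sound class with a basic foley attenuation that limits its audible range.
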SoundCue /HighTower/Audio/Grape/HightowerGrape_BrambleShield_Impact_Cue
root=Random_0
nodes [
 {
  id=Random_0
  children [WavePlayer_0,WavePlayer_1,WavePlayer_2,WavePlayer_3]
  properties {bShouldExcludeFromBranchCulling=true, weights [1,1,1,1]}
 }
 {id=WavePlayer_0, properties {soundWaveAssetPtr=&/HightowerGrape_BrambleShield_Impact_01}}
 {id=WavePlayer_1, properties {soundWaveAssetPtr=&/HightowerGrape_BrambleShield_Impact_02}}
 {id=WavePlayer_2, properties {soundWaveAssetPtr=&/HightowerGrape_BrambleShield_Impact_03}}
 {id=WavePlayer_3, properties {soundWaveAssetPtr=&/HightowerGrape_BrambleShield_Impact_04}}
]
properties {
 attenuationSettings=&Foley_Attenuation_Basic_Att
 maxDistance=1536
 soundClassObject=&Fort_Gadgets
 volumeMultiplier=0.3
}
