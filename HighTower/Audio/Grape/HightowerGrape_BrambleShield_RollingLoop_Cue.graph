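# Looping Sound Cue for the rolling sound of the HightowerGrape Bramble Shield.
# A looping Wave Player feeds a Continuous Modulator, which takes a runtime Speed parameter for both volume and pitch, each with its own mapping.
# The cue plays in the Fort_Gadgets sound class at reduced volume, using shared basic foley attenuation.
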SoundCue /HighTower/Audio/Grape/HightowerGrape_BrambleShield_RollingLoop_Cue
root=ModulatorContinuous_0
nodes [
 {
  id=ModulatorContinuous_0
  children [WavePlayer_0]
  properties {
   pitchModulationParams {default=0, minInput=1, minOutput=1, parameterName=Speed}
   volumeModulationParams {default=0, minInput=0.3, minOutput=0.3, parameterName=Speed}
  }
 }
 {id=WavePlayer_0, properties {bLooping=true, soundWaveAssetPtr=&/HightowerGrape_BrambleShield_RollingLoop}}
]
properties {
 attenuationSettings=&Foley_Attenuation_Basic_Att
 maxDistance=1536
 soundClassObject=&Fort_Gadgets
 volumeMultiplier=0.6
}
